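; Pawn Blueprint for the WildEstate BuzzCool NPC (despite the AnimBP name), handling mostly cosmetic combat feedback.
; Events drive creature impact sound and particles, sound-library pain audio, Niagara bug-death FX, custom depth updates and disabled overlaps on clients.
; Helpers cache awaken and building-hit FX components per static mesh and push material parameters to main, duplicate and auxiliary meshes.
; Server death calls SpawnActorOnDeath, which is empty here, as are the skeletal-mesh FX functions.
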(event Custom|ForceUpdateCustomDepth
  (bind _fortpawncomponent_customdepth (Variables|Default|GetFortPawnComponent_CustomDepth))
  (if (CustomDepth|IsRenderCustomDepthEnabled _fortpawncomponent_customdepth)
    (CustomDepth|Update _fortpawncomponent_customdepth 0.0 true)))

(event EventBeginPlay
  ;[1 1 1] Begin Play
  (Utilities|FlowControl|Sequence
    (:then_0
      (Networking|SwitchHasAuthority
        (:Remote
          (bind _asprimitive_component (Utilities|Casting|CastToPrimitiveComponent (Variables|/Script/Engine.MovementComponent|GetUpdatedComponent (Variables|Character|GetCharacterMovement))))
          (Variables|/Script/Engine.PrimitiveComponent|SetbGenerateOverlapEvents false _asprimitive_component))))))

(event Custom|UpdateCustomDepth
  (bind _fortpawncomponent_customdepth (Variables|Default|GetFortPawnComponent_CustomDepth))
  ;[1 1 1] Begin Play
  (if (CustomDepth|IsRenderCustomDepthEnabled _fortpawncomponent_customdepth)
    (CustomDepth|Update _fortpawncomponent_customdepth (Variables|Default|GetCustom%20Depth%20Update%20Rate) false)))

(event GameplayCue|GameplayCue.Impact.Physical.Creature
  (params (EventType :type EGameplayCueEvent%20Enum :default "OnActive") (Parameters :type Gameplay%20Cue%20Parameters%20Structure%20%28by%20ref%29))
  (bind (_bblockinghit _binitialoverlap _time _distance _location) (Collision|BreakHitResult (Ability|GameplayCue|GetHitResult Parameters)))
  ;[1 1 1] ImpactATarget
  (Audio|PlaySoundatLocation (Sound|GetImpactPhysicalSurfaceSound (Ability|GameplayCue|GetHitResult Parameters)) _location)
  (Default|PlayDamageFXSurfaceEnemies Parameters (Variables|Effects|GetImpactParticles_Default) (Variables|Effects|GetImpactParticles_Wood) (Variables|Effects|GetImpactParticles_Stone) (Variables|Effects|GetImpactParticles_Metal) (Variables|Effects|GetImpactParticles_HumanFlesh) (.location (Transformation|GetSocketTransform (Variables|Character|GetMesh) (Variables|Default|GetImpact_MeleeSocket) "RTS_World")) (.rotation (Transformation|GetSocketTransform (Variables|Character|GetMesh) (Variables|Default|GetImpact_MeleeSocket) "RTS_World")) (Variables|Default|GetMeleeImpact_UseSocketTransforms)))

(event EventOnDeathPlayEffects
  (params (Damage :type Float%20%28single-precision%29) (DamageTags :type Gameplay%20Tag%20Container%20Structure%20%28by%20ref%29) (Momentum :type Vector :default "0, 0, 0") (HitInfo :type Hit%20Result%20Structure%20%28by%20ref%29) (InstigatedBy :type Fort%20Pawn%20Object%20Reference) (DamageCauser :type Actor%20Object%20Reference) (EffectContext :type Gameplay%20Effect%20Context%20Handle%20Structure))
  (bind (_bblockinghit _binitialoverlap _time _distance _location _impactpoint _normal _impactnormal) (Collision|BreakHitResult HitInfo))
  ;[1 1 1] On Death Effects
  (Lighting|SetCastCapsuleIndirectShadow (Variables|Character|GetMesh) false)
  (bind _returnvalue (Niagara|SpawnSystematLocation (Variables|Default|GetFX_BugDeath) (+ (Transformation|GetActorLocation) "0.000000,0.000000,10.000000") (Math|Rotator|MakeRotfromY _impactnormal))))

(event EventOnDeathServer
  (params (Damage :type Float%20%28single-precision%29) (DamageTags :type Gameplay%20Tag%20Container%20Structure%20%28by%20ref%29) (Momentum :type Vector :default "0, 0, 0") (HitInfo :type Hit%20Result%20Structure%20%28by%20ref%29) (InstigatedBy :type Controller%20Object%20Reference) (DamageCauser :type Actor%20Object%20Reference) (EffectContext :type Gameplay%20Effect%20Context%20Handle%20Structure))
  ;[1 1 1] On Death (Server)
  (|CallParentFunction Damage DamageTags Momentum HitInfo InstigatedBy DamageCauser EffectContext)
  (CallFunction|SpawnActoronDeath))

(event EventOnDamagePlayEffects
  (params (Damage :type Float%20%28single-precision%29) (DamageTags :type Gameplay%20Tag%20Container%20Structure%20%28by%20ref%29) (Momentum :type Vector :default "0, 0, 0") (HitInfo :type Hit%20Result%20Structure%20%28by%20ref%29) (InstigatedBy :type Fort%20Pawn%20Object%20Reference) (DamageCauser :type Actor%20Object%20Reference) (EffectContext :type Gameplay%20Effect%20Context%20Handle%20Structure))
  ;[1 1 1] Hit Effects
  (CallFunction|PlaySoundLibraryAudio "(TagName=\"SoundLibrary.Deimos.Pain\")"))

(fn ApplyAwakenEffectToSkeletalMesh
  (params (Mesh :type Skeletal%20Mesh%20Component%20Object%20Reference)))

(fn ApplyAwakenEffectToStaticMesh
  (params (Mesh :type Static%20Mesh%20Component%20Object%20Reference))
  (bind _awakenfxstaticmeshes (Variables|Effects|GetAwakenFXStaticMeshes))
  (bind (_value _returnvalue) (Utilities|Map|Find _awakenfxstaticmeshes Mesh))
  (bind _fxstaticmesh (LocalVariables|GetFXStaticMesh))
  (Utilities|IsValid Mesh
    (:Is%20Valid
      (if _returnvalue
        (bind _output_get (LocalVariables|SetFXStaticMesh _value))
        (Utilities|IsValid _output_get
          (:Is%20Valid
            _)
          (:Is%20Not%20Valid
            (bind _returnvalue_2 (AddComponent|NODE_AddFXStaticMeshComponent-2))
            (Utilities|Map|Add (Variables|Effects|GetAwakenFXStaticMeshes) Mesh _returnvalue_2)
            (bind _output_get_2 (LocalVariables|SetFXStaticMesh _returnvalue_2))
            (Class|FortFXStaticMeshComponent|SetSource _output_get_2 Mesh)
            (Variables|/Script/FortniteGame.FortFXStaticMeshComponent|SetAwakenDuration (Variables|Effects|GetAwakenDuration) _fxstaticmesh)))
        (Class|FXStaticMeshComponent|ApplyAwakenEffect _fxstaticmesh)
        (else
          _)))))

(fn ApplyBuildingHitEffectToSkeletalMesh
  (params (Mesh :type Skeletal%20Mesh%20Component%20Object%20Reference)))

(fn ApplyBuildingHitEffectToStaticMesh
  (params (Mesh :type Static%20Mesh%20Component%20Object%20Reference))
  (bind _buildinghitfxstaticmeshes (Variables|Effects|GetBuildingHitFXStaticMeshes))
  (bind (_value _returnvalue) (Utilities|Map|Find _buildinghitfxstaticmeshes Mesh))
  (bind _fxstaticmesh (LocalVariables|GetFXStaticMesh))
  (Utilities|IsValid Mesh
    (:Is%20Valid
      (if _returnvalue
        (bind _output_get (LocalVariables|SetFXStaticMesh _value))
        (Utilities|IsValid _output_get
          (:Is%20Valid
            _)
          (:Is%20Not%20Valid
            (bind _returnvalue_2 (AddComponent|NODE_AddFXStaticMeshComponent-0))
            (Utilities|Map|Add (Variables|Effects|GetBuildingHitFXStaticMeshes) Mesh _returnvalue_2)
            (bind _output_get_2 (LocalVariables|SetFXStaticMesh _returnvalue_2))
            (Class|FortFXStaticMeshComponent|SetSource _output_get_2 Mesh)
            (Variables|/Script/FortniteGame.FortFXStaticMeshComponent|SetBuildingHitDuration (Variables|Effects|GetBuildingHitDuration) _fxstaticmesh)))
        (Class|FXStaticMeshComponent|ApplyBuildingHitEffect _fxstaticmesh)
        (else
          _)))))

(fn FindCloseGroundPointOnNavMesh
  (bind _self self)
  (bind _returnvalue (Transformation|GetActorLocation _self))
  (bind (_outhit _returnvalue_2) (Collision|LineTraceByChannel _returnvalue "0.000000,0.000000,-10000.000000" "TraceTypeQuery10" true))
  (if _returnvalue_2
    (LocalVariables|SetIsWorldTraceHit true)
    (LocalVariables|SetWorldTraceLocation (.Location (Collision|BreakHitResult _outhit)))
    (else
      (LocalVariables|SetIsWorldTraceHit false)
      (LocalVariables|SetWorldTraceLocation _returnvalue)))
  (return (LocalVariables|GetWorldTraceLocation) (LocalVariables|GetIsWorldTraceHit)))

(fn Orphaned
  (return false 0))

(fn PlayAdditiveHitReacts
  (params (Hit%20Direction :type Vector) (Anim%20Montage :type Anim%20Montage%20Object%20Reference))
  (HitReacts|PlayAdditiveHitReactsNative Hit%20Direction Anim%20Montage) ; This used to be done directly in the Event graph, which meant any Child of this Pawn couldn't override the Function behavior.\n\nWith this, we should be able to override PlayAdditiveHitReacts and call the Native function here with different logic in front of it, if we need to.
)

(fn PlaySoundLibraryAudio
  (params (Event%20Name :type Gameplay%20Tag%20Structure))
  (Networking|CanExecuteCosmeticEvents
    (:True
      (if (Utilities|IsValid (Variables|Sound|GetSoundLibraryContext) :!class /Script/Engine.KismetSystemLibrary)
        (Utilities|Struct|SetmembersinSimpleSoundSettings (Variables|/Script/SoundLibrary.SoundLibrarySimpleContext|GetSettings (Variables|Sound|GetSoundLibraryContext)) Event%20Name)
        (else
          (bind _returnvalue (SoundLibrary|CreateSoundLibraryContext :Class "/Script/SoundLibrary.SoundLibrarySimpleContext" :Settings (Utilities|Struct|MakeSimpleSoundSettings Event%20Name) :Actor self :bForwardToActorOwner true :bForwardToAttachedActors true :bAutoResetContext true))
          (bind _output_get (Variables|Sound|SetSoundLibraryContext _returnvalue))))
      (SoundLibrary|PlaySoundwithContext (WorldSubsystems|GetSoundLibrarySubsystem) _output_get))))

(fn SetScalarParameterOnAllCharacterMIDs
  (params (Parameter%20Name :type Name) (Scalar%20Value :type Float%20%28double-precision%29) (AlsoModifyOriginalMIDs :type Boolean))
  (bind _auxillary_objects (Variables|Effects|GetAuxillary%20Objects))
  (bind _duplicate_auxillary_objects (Variables|Effects|GetDuplicate%20Auxillary%20Objects))
  (Utilities|FlowControl|Sequence
    (:then_0
      (bind _mesh (Variables|Character|GetMesh)
        (:then
          (Rendering|Material|SetScalarParameterValueonMaterials _mesh (Variables|GetParameter%20Name) (Variables|GetScalar%20Value)))
        (:else
          _))
      (bind _duplicatecharactermesh (Variables|Effects|GetDuplicateCharacterMesh)
        (:then
          (Rendering|Material|SetScalarParameterValueonMaterials _duplicatecharactermesh (Variables|GetParameter%20Name) (Variables|GetScalar%20Value)))
        (:else)))
    (:then_1
      (if (Utilities|IsArrayNonEmpty _auxillary_objects)
        (for _array_element _auxillary_objects
          (Utilities|IsValid _array_element
            (:Is%20Valid
              (bind _asmesh_component (Utilities|Casting|CastToMeshComponent _array_element))
              (Rendering|Material|SetScalarParameterValueonMaterials _asmesh_component (Variables|GetParameter%20Name) (Variables|GetScalar%20Value)))))))
    (:then_2
      (if (Utilities|IsArrayNonEmpty _duplicate_auxillary_objects)
        (for _array_element_2 _duplicate_auxillary_objects
          (Utilities|IsValid _array_element_2
            (:Is%20Valid
              (bind _asmesh_component_2 (Utilities|Casting|CastToMeshComponent _array_element_2))
              (Rendering|Material|SetScalarParameterValueonMaterials _asmesh_component_2 (Variables|GetParameter%20Name) (Variables|GetScalar%20Value)))))))
    (:then_3
      ;[1 1 1] Optionally keep the original MIDs in sync with the override MIDs
      (if AlsoModifyOriginalMIDs
        (Utilities|FlowControl|Sequence
          (:then_0
            (for _array_element_3 (Variables|Materials|GetPrevious%20MID)
              (Rendering|Material|SetScalarParameterValue _array_element_3 (Variables|GetParameter%20Name) (Variables|GetScalar%20Value) :!class /Script/Engine.MaterialInstanceDynamic)))
          (:then_1
            (for _array_element_4 (Variables|Materials|GetPrevious%20MID_AuxObjs)
              (Rendering|Material|SetScalarParameterValue _array_element_4 (Variables|GetParameter%20Name) (Variables|GetScalar%20Value) :!class /Script/Engine.MaterialInstanceDynamic))))))))

(fn SetVectorParameterOnAllCharacterMIDs
  (params (Parameter%20Name :type Name) (Vector%20Value :type Vector) (Propagate%20to%20Auxiliary%20Meshes :type Boolean :default true))
  (bind _auxillary_objects (Variables|Effects|GetAuxillary%20Objects))
  (bind _duplicate_auxillary_objects (Variables|Effects|GetDuplicate%20Auxillary%20Objects))
  (bind _mesh (Variables|Character|GetMesh)
    (:then
      (Rendering|Material|SetVectorParameterValueonMaterials _mesh Parameter%20Name Vector%20Value))
    (:else
      _))
  (bind _duplicatecharactermesh (Variables|Effects|GetDuplicateCharacterMesh)
    (:then
      (Rendering|Material|SetVectorParameterValueonMaterials _duplicatecharactermesh Parameter%20Name Vector%20Value))
    (:else
      _))
  (if (and (> (Utilities|Array|Length _auxillary_objects) 0) (Variables|GetPropagate%20to%20Auxiliary%20Meshes))
    (for _array_element _auxillary_objects
      (Utilities|IsValid _array_element
        (:Is%20Valid
          (bind _asmesh_component (Utilities|Casting|CastToMeshComponent _array_element))
          (Rendering|Material|SetVectorParameterValueonMaterials _asmesh_component (Variables|GetParameter%20Name) (Variables|GetVector%20Value)))))
    (if (and (> (Utilities|Array|Length _duplicate_auxillary_objects) 0) (Variables|GetPropagate%20to%20Auxiliary%20Meshes))
      (for _array_element_2 _duplicate_auxillary_objects
        (Utilities|IsValid _array_element_2
          (:Is%20Valid
            (bind _asmesh_component_2 (Utilities|Casting|CastToMeshComponent _array_element_2))
            (Rendering|Material|SetVectorParameterValueonMaterials _asmesh_component_2 (Variables|GetParameter%20Name) (Variables|GetVector%20Value)))))))
  (return))

(fn SpawnActorOnDeath)

(fn SpawnParticleSystemOnCharacterMesh
  (params (ParticleSystemTemplate :type Particle%20System%20Object%20Reference) (ParticleSystemComponentReferenceVar :type Cascade%20Particle%20System%20Component%20Object%20Reference) (AttachPointName :type Name) (Location :type Vector) (Rotation :type Rotator) (InstanceParameters :type Array%20of%20Particle%20Sys%20Param%20Structures) (AutoActivate :type Boolean :default true) (AutoDestroy :type Boolean) (AbsoluteLocation :type Boolean) (AbsoluteRotation :type Boolean) (AbsoluteScale :type Boolean))
  (Networking|CanExecuteCosmeticEvents
    (:True
      (Utilities|IsValid ParticleSystemTemplate
        (:Is%20Valid
          (Utilities|IsValid ParticleSystemComponentReferenceVar
            (:Is%20Valid
              (Components|Activation|Deactivate ParticleSystemComponentReferenceVar))
            (:Is%20Not%20Valid
              _))
          (bind _returnvalue (Effects|SpawnEmitterAttached ParticleSystemTemplate (Variables|Character|GetMesh) AttachPointName Location Rotation "1.000000,1.000000,1.000000" "KeepRelativeOffset" AutoDestroy))
          (Variables|/Script/CascadeRuntime.ParticleSystemComponent|SetInstanceParameters InstanceParameters _returnvalue)
          (Components|Activation|SetActive _returnvalue AutoActivate false)
          (Transformation|SetAbsolute _returnvalue AbsoluteLocation AbsoluteRotation AbsoluteScale)
          (return _returnvalue))))))

(fn ConstructionScript
  (Networking|CanExecuteCosmeticEvents
    (:True
      (AI|SetWeaponCollisionComponent (Variables|Default|GetWeaponCapsuleCollision))
      (bind _returnvalue (Rendering|Material|GetScalarParameterValue "/Game/Packages/Fortress_SharedMaterials/GlobalMaterialParameters/FortniteMaterialParameters.FortniteMaterialParameters" "Enemy Indirect Shadow Min Occlusion" :!class /Script/Engine.KismetMaterialLibrary))
      (Lighting|SetCapsuleIndirectShadowMinVisibility (Variables|Character|GetMesh) _returnvalue))))
